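# Sound cue for the destruction sound of a WeirdObject, filed under player gameplay sounds.
# A single wave player feeds an attenuation node that borrows the Portal ambient attenuation asset.
# The cue uses a ducking impact sound class and a boosted volume.
# Its own concurrency override caps simultaneous voices and stops the quietest when the limit is reached.
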SoundCue /Figment_S09/Sounds/Fort_GamePlay_Sounds/Player/WeirdObject/WeirdObject_Destroy_Cue
root=Attenuation_0
nodes [
 {
  id=Attenuation_0
  children [WavePlayer_0]
  properties {attenuationSettings=/Figment_S09/Sounds/Fort_GamePlay_Sounds/Player/Portal/Portal_Ambient_Att}
 }
 {id=WavePlayer_0, properties {soundWaveAssetPtr=&/WeirdObjects_Destroy_01}}
]
properties {
 bOverrideConcurrency=true
 maxDistance=2000
 soundClassObject=/Game/Sounds/FortSoundClassesAndModes/Fort_ImpactSounds_DuckDefaultQuieter_PSM
 volumeMultiplier=1.75
 attenuationOverrides {
  bEnableFocusInterpolation=true
  bEnableListenerFocus=true
  distanceAlgorithm=Logarithmic
  falloffDistance=2000
  nonFocusVolumeAttenuation=0.5
  stereoSpread=512
  attenuationShapeExtents {x=0}
 }
 concurrencyOverrides {maxCount=8, resolutionRule=StopQuietest}
}
