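# Looping in-world ambient Sound Cue for the WeirdObject, routed to the gadgets sound class.
# A mixer layers two looping waves, each through its own attenuation node, with the second at half volume.
# The first layer uses the WeirdObject attenuation asset, while the second overrides its attenuation inline with a short falloff, listener focus and occlusion filter settings.
# A concurrency override caps simultaneous instances and stops the quietest.
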SoundCue /Figment_S09/Sounds/Fort_GamePlay_Sounds/Player/WeirdObject/WeirdObject_InWorld_Loop_Cue
root=Mixer_1
nodes [
 {
  id=Attenuation_0
  children [WavePlayer_4]
  properties {attenuationSettings=&/WeirdObject_Att}
 }
 {
  id=Attenuation_1
  children [WavePlayer_5]
  properties {
   attenuationSettings=/Figment_S09/Sounds/Fort_GamePlay_Sounds/Player/Portal/Portal_Ambient_Att
   bOverrideAttenuation=true
   attenuationOverrides {
    bEnableFocusInterpolation=true
    bEnableListenerFocus=true
    falloffDistance=512
    focusAzimuth=15
    nonFocusAzimuth=25
    nonFocusVolumeAttenuation=0.5
    occlusionInterpolationTime=0.5
    occlusionLowPassFilterFrequency=4000
    occlusionTraceChannel=ECC_GameTraceChannel6
    stereoSpread=256
    attenuationShapeExtents {x=0}
   }
  }
 }
 {id=Mixer_1, children [Attenuation_0, Attenuation_1], properties {inputVolume [1,0.5]}}
 {
  id=WavePlayer_4
  properties {bLooping=true, soundWaveAssetPtr=&/WeirdObject_Ambient_Loop_01}
 }
 {
  id=WavePlayer_5
  properties {bLooping=true, soundWaveAssetPtr=&/WeirdObject_Ambient_Loop_02}
 }
]
properties {
 bOverrideConcurrency=true
 maxDistance=2000
 soundClassObject=&Fort_Gadgets
 attenuationOverrides {
  bEnableFocusInterpolation=true
  bEnableListenerFocus=true
  distanceAlgorithm=Logarithmic
  falloffDistance=2000
  nonFocusVolumeAttenuation=0.5
  stereoSpread=512
  attenuationShapeExtents {x=0}
 }
 concurrencyOverrides {maxCount=8, resolutionRule=StopQuietest}
}
