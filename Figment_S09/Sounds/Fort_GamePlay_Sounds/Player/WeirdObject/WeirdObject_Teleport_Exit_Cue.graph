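# Sound cue for the WeirdObject gadget's teleport exit, in the player gameplay sounds.
# A random node picks among three equally weighted wave players, which use the third-person "Enter" recordings despite the cue's Exit name.
# It plays in the gadgets sound class at slightly reduced volume, spatialized by a shared third-person portal enter/exit attenuation asset.
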SoundCue /Figment_S09/Sounds/Fort_GamePlay_Sounds/Player/WeirdObject/WeirdObject_Teleport_Exit_Cue
root=Random_0
nodes [
 {id=Random_0, children [WavePlayer_1,WavePlayer_2,WavePlayer_3], properties {weights [1,1,1]}}
 {id=WavePlayer_1, properties {soundWaveAssetPtr=&/WeirdObject_Enter_3P_01}}
 {id=WavePlayer_2, properties {soundWaveAssetPtr=&/WeirdObject_Enter_3P_02}}
 {id=WavePlayer_3, properties {soundWaveAssetPtr=&/WeirdObject_Enter_3P_03}}
]
properties {
 attenuationSettings=/Game/Sounds/Attenuation/Fort_Portal_EnterExit_3P_Athena_Att
 maxDistance=3584
 soundClassObject=&Fort_Gadgets
 volumeMultiplier=0.8
}
