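# Sound cue named as an ambient loop for weird objects, filed under player gameplay sounds, though it has no nodes or root, so it references no audio.
# It carries only playback settings: reduced volume, the dance music quest sound class and overridden attenuation.
# The attenuation uses a box shape with logarithmic falloff, a non-spatialized radius near the source, and low-pass occlusion traced on a game-specific channel.
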SoundCue /Figment_S09/Sounds/Fort_GamePlay_Sounds/Player/WeirdObject/WeirdObjects_Ambient_Loop_Cue
root=null
nodes []
properties {
 bOverrideAttenuation=true
 maxDistance=1536
 soundClassObject=/Game/Sounds/FortSoundClassesAndModes/Fort_DanceMusic_Quest_PSM
 volumeMultiplier=0.3
 attenuationOverrides {
  attenuationShape=Box
  bEnableOcclusion=true
  distanceAlgorithm=Logarithmic
  falloffDistance=1024
  nonSpatializedRadiusStart=1024
  occlusionInterpolationTime=0.5
  occlusionLowPassFilterFrequency=4000
  occlusionTraceChannel=ECC_GameTraceChannel11
  stereoSpread=1024
  attenuationShapeExtents=<512 512 384>
 }
}
